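; Quest gameplay ability that credits a player who already captured a POI this match before receiving the quest.
; On activation it runs a one-time check for the captured-POI tag, a banner GameplayCue that persists on the pawn all match.
; If the tag is present, it fakes a claim by sending a PreviouslyClaimedPOI custom stat event to the quest manager, then ends itself.
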(event Ability|EventFailedToActivatePassiveAbility
  (params (ActorInfo :type Gameplay%20Ability%20Actor%20Info%20Structure))
  ;[1 1 1] Check if player has already captured a POI this match. LS 4/24/23
  (|CallParentFunction ActorInfo))

(event Ability|EventActivateAbility
  ;[1 1 1] Check if player has already captured a POI this match. LS 4/24/23
  (|CallParentFunction)
  ;;[1 1 1] Initial tag check when we get the quest.
  (Utilities|FlowControl|DoOnce false
    (:Completed
      ;;[1 1 1] Checking for GameplayCue.POIBanner.Player.Capturing which persists on the player all match if they've claimed a point.
      (if (GameplayTags|HasMatchingGameplayTag (Variables|GAVars|Data|GetOwningPlayerPawn) (Variables|Default|GetCaptured%20POI%20Already))
        ;;[1 1 1] Fake firing the POIClaimed tag
        (Quest|SendComplexCustomStatEvent (Variables|GAVars|Data|GetOwningPlayerQuestManager) 0 "(GameplayTags=((TagName=\"Athena.Quest.PreviouslyClaimedPOI\")))" "(GameplayTags=)" 1)
        (Ability|EndAbility)))))

(fn GetOwnedGameplayTags
  (params (self2 :type Gameplay%20Tag%20Asset%20Interface%20Interface))
  (return (GameplayTags|GetOwnedGameplayTags self2 :!class /Script/GameplayTags.GameplayTagAssetInterface)))
